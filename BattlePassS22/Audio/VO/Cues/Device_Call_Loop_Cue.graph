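# Battle Pass S22 device-call audio cue that plays an answer burst followed by continuous static, routed to the non-critical in-game UI sound class.
# After a fixed delay, a mixer layers an answer chirp, a random chirp or pop, and looping static that fades in.
# Two looping layers add randomly delayed, pitch- and volume-varied pops over the static.
# Quality Level nodes supply these pop layers only in the first two of four quality slots.
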SoundCue /BattlePassS22/Audio/VO/Cues/Device_Call_Loop_Cue
root=Delay_2
nodes [
 {id=Delay_0, children [Modulator_1], properties {delayMax=10, delayMin=2}}
 {id=Delay_1, children [Modulator_2], properties {delayMax=15, delayMin=5}}
 {id=Delay_2, children [Mixer_0], properties {delayMax=1.35, delayMin=1.35}}
 {
  id=Enveloper_0
  children [WavePlayer_0]
  properties {
   pitchCurve {editorCurveData {keys [{time=0}]}}
   volumeCurve {editorCurveData {keys [{arriveTangent=2.7426, interpMode=RCIM_Cubic, leaveTangent=2.7426, tangentMode=RCTM_User, time=0, value=0}, <at 0.5 1>]}}
  }
 }
 {id=Looping_0, children [Delay_0]}
 {id=Looping_1, children [Delay_1]}
 {id=Mixer_0, children [Modulator_0, Modulator_3, Enveloper_0, QualityLevel_0, QualityLevel_1], properties {inputVolume [2,4,1,0.5,0.5]}}
 {id=Modulator_0, children [Random_0]}
 {id=Modulator_1, children [Random_1], properties {pitchMax=2, volumeMax=1, volumeMin=0.5}}
 {id=Modulator_2, children [Random_2], properties {pitchMax=0.6, pitchMin=0.5, volumeMax=0.6, volumeMin=0}}
 {id=Modulator_3, children [WavePlayer_10]}
 {id=QualityLevel_0, children [Looping_0,Looping_0,null,null]}
 {id=QualityLevel_1, children [Looping_1,Looping_1,null,null]}
 {id=Random_0, children [WavePlayer_1,WavePlayer_2,WavePlayer_3], properties {bShouldExcludeFromBranchCulling=true, weights [1,1,1]}}
 {id=Random_1, children [WavePlayer_4,WavePlayer_5,WavePlayer_6], properties {bShouldExcludeFromBranchCulling=true, weights [1,1,1]}}
 {id=Random_2, children [WavePlayer_7,WavePlayer_8,WavePlayer_9], properties {bShouldExcludeFromBranchCulling=true, weights [1,1,1]}}
 {id=WavePlayer_0, properties {bLooping=true, soundWaveAssetPtr=/BattlePassS22/Audio/VO/Cues/Device_Loop/Device_Call_Static_Loop_01}}
 {id=WavePlayer_1, properties {soundWaveAssetPtr=/BattlePassS22/Audio/VO/Cues/Device_Loop/Device_Call_Incoming_Answer_Chirp_01}}
 {id=WavePlayer_10, properties {soundWaveAssetPtr=&Device_Call_Incoming_Answer_Chirp_01}}
 {id=WavePlayer_2, properties {soundWaveAssetPtr=/BattlePassS22/Audio/VO/Cues/Device_Loop/Device_Call_Incoming_Pop_02}}
 {id=WavePlayer_3, properties {soundWaveAssetPtr=/BattlePassS22/Audio/VO/Cues/Device_Loop/Device_Call_Incoming_Pop_03}}
 {id=WavePlayer_4, properties {soundWaveAssetPtr=/BattlePassS22/Audio/VO/Cues/Device_Loop/Device_Call_Incoming_Pop_01}}
 {id=WavePlayer_5, properties {soundWaveAssetPtr=&Device_Call_Incoming_Pop_02}}
 {id=WavePlayer_6, properties {soundWaveAssetPtr=&Device_Call_Incoming_Pop_03}}
 {id=WavePlayer_7, properties {soundWaveAssetPtr=&Device_Call_Incoming_Pop_01}}
 {id=WavePlayer_8, properties {soundWaveAssetPtr=&Device_Call_Incoming_Pop_02}}
 {id=WavePlayer_9, properties {soundWaveAssetPtr=&Device_Call_Incoming_Pop_03}}
]
properties {maxDistance=2097152, soundClassObject=&Fort_UI_InGame_NonCritical, volumeMultiplier=0.5}
